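; Emote prop Blueprint for the Basil Strong emote that drives a camera shake value for the locally viewed pawn outside the frontend.
; On BeginPlay it waits briefly, casts its attach parent to a FortPawn and, if that pawn is locally viewed, plays a timeline.
; The timeline writes the CameraShake scalar in the shared MPC_Galileo material parameter collection, which is reset when the prop is destroyed.
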(event EventBeginPlay
  (if (not (Reactive|Frontend|IsCosmeticObjectDisplayedinFrontendContext 0))
    (Utilities|FlowControl|Delay 0.1)
    (bind _asfort_pawn (Utilities|Casting|CastToFortPawn (Actor|GetAttachParentActor self)))
    (if (Pawn|IsLocallyViewed _asfort_pawn)
      (bind (_direction _camerashake) (Timeline|Timeline PlayFromStart)
        (:Update
          (Rendering|Material|SetScalarParameterValue "/Game/Athena/Prototype/Blueprints/Galileo/Blueprints/MPC_Galileo.MPC_Galileo" "CameraShake" _camerashake :!class /Script/Engine.KismetMaterialLibrary))))))

(event EventDestroyed
  (Rendering|Material|SetScalarParameterValue "/Game/Athena/Prototype/Blueprints/Galileo/Blueprints/MPC_Galileo.MPC_Galileo" "CameraShake" :!class /Script/Engine.KismetMaterialLibrary)
  (Actor|DestroyActor))

(fn ConstructionScript)
